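# Sound Cue for a looping daytime interior room-tone ambience in a large chaparral villa room on the Helios starter island.
# A Quality Level node picks a looping medium stone room tone, quad for the first two levels and stereo for the last two, each through a Mixer.
# Two nested, endlessly looping Envelopers add slow drift: a 30-second pitch curve inside a 20-second gentle volume dip.
# It plays quietly with ambient attenuation and an outdoor ambient sound class.
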SoundCue /Game/Sounds/Helios/POI/StarterIsland/Ambience_Chapparel_Villa_Large_Room_Int_02_Day_Loop_Cue
root=Enveloper_3
nodes [
 {
  id=Enveloper_0
  children [QualityLevel_1]
  properties {
   bLoop=true
   bLoopIndefinitely=true
   loopEnd=30
   pitchCurve {editorCurveData {keys [{interpMode=RCIM_Cubic, time=0}, <at 15.9468 0.9181 RCIM_Cubic>, {interpMode=RCIM_Cubic, time=30}]}}
   volumeCurve {editorCurveData {keys [<at 0 1>]}}
  }
 }
 {
  id=Enveloper_3
  children [Enveloper_0]
  properties {
   bLoop=true
   bLoopIndefinitely=true
   loopEnd=20
   pitchCurve {editorCurveData {keys [{time=0}]}}
   volumeCurve {editorCurveData {keys [
      {arriveTangent=0, interpMode=RCIM_Cubic, leaveTangent=0, time=0, value=1}
      {arriveTangent=0, interpMode=RCIM_Cubic, leaveTangent=0, time=10, value=0.9}
      {arriveTangent=0, interpMode=RCIM_Cubic, leaveTangent=0, time=20, value=0.9992}
     ]}}
  }
 }
 {id=Mixer_1, children [null,WavePlayer_2], properties {inputVolume [0.3,1]}}
 {id=Mixer_3, children [WavePlayer_3,null], properties {inputVolume [1,0.3]}}
 {id=QualityLevel_1, children [Mixer_1,Mixer_1,Mixer_3,Mixer_3]}
 {
  id=WavePlayer_2
  properties {bLooping=true, soundWaveAssetPtr=/Game/Sounds/Artemis/RoomTones/Stone/Artemis_Interior_Medium_Stone_Room_01_Quad}
 }
 {
  id=WavePlayer_3
  properties {bLooping=true, soundWaveAssetPtr=/Game/Sounds/Artemis/RoomTones/Stone/Artemis_Interior_Medium_Stone_Room_01_Stereo}
 }
]
properties {
 attenuationSettings=&Fort_Ambient_Athena_Att
 maxDistance=2097152
 soundClassObject=&Fort_Ambient_Outdoors_Athena
 volumeMultiplier=0.2
 preEffectBusSends [{}]
}
