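# Looping ambient cue for gentle ocean shoreline water at the Helios Starter Island POI, limited to one instance.
# After a short random delay and volume fade-in, a mixer layers gentle laps through very-close and close attenuations with quieter distant waves through medium and distant attenuations.
# The medium attenuation chains into the distant one, so both branches share the wave source.
# Both loops are pitched down, and looping envelopes slowly modulate the waves' pitch and volume.
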SoundCue /Game/Sounds/Helios/POI/StarterIsland/Helios_Water_Ocean_Shore_Gentle_03_StarterIsland_Cue
root=Delay_0
nodes [
 {
  id=Attenuation_1
  children [Modulator_2]
  properties {attenuationSettings=/Game/Sounds/Asteria/Ocean/Attenuations/Asteria_Water_Ocean_Shore_Close_Att}
 }
 {
  id=Attenuation_7
  children [Attenuation_8]
  properties {attenuationSettings=/Game/Sounds/Asteria/Ocean/Attenuations/Asteria_Water_Ocean_Shore_Medium_Att}
 }
 {
  id=Attenuation_8
  children [Enveloper_3]
  properties {attenuationSettings=/Game/Sounds/Asteria/Ocean/Attenuations/Asteria_Water_Ocean_Shore_Distant_Att}
 }
 {
  id=Attenuation_9
  children [Modulator_2]
  properties {attenuationSettings=/Game/Sounds/Asteria/Ocean/Attenuations/Asteria_Water_Ocean_Shore_Very_Close_Att}
 }
 {id=Delay_0, children [Enveloper_1], properties {delayMax=1, delayMin=0.2}}
 {
  id=Enveloper_0
  children [Modulator_3]
  properties {
   bLoop=true
   bLoopIndefinitely=true
   loopEnd=30
   pitchCurve {editorCurveData {keys [
      {arriveTangent=-0.0013, interpMode=RCIM_Cubic, leaveTangent=-0.0013, time=0}
      <at 15 0.98 RCIM_Cubic>
      {arriveTangent=0.0013, interpMode=RCIM_Cubic, leaveTangent=0.0013, time=30}
     ]}}
   volumeCurve {editorCurveData {keys [<at 0 1>]}}
  }
 }
 {
  id=Enveloper_1
  children [Mixer_2]
  properties {pitchCurve {editorCurveData {keys [{time=0}]}}, volumeCurve {editorCurveData {keys [<at 0 0>,<at 0.5 1>]}}}
 }
 {
  id=Enveloper_3
  children [Enveloper_0]
  properties {
   bLoop=true
   bLoopIndefinitely=true
   loopEnd=20
   pitchCurve {editorCurveData {keys [{time=0}]}}
   volumeCurve {editorCurveData {keys [
      {arriveTangent=-0.01, interpMode=RCIM_Cubic, leaveTangent=-0.01, time=0, value=1}
      <at 10 0.9 RCIM_Cubic>
      {arriveTangent=0.01, interpMode=RCIM_Cubic, leaveTangent=0.01, time=20, value=1}
     ]}}
  }
 }
 {id=Mixer_0, children [WavePlayer_2,null], properties {inputVolume [1,1]}}
 {id=Mixer_2, children [Attenuation_9, Attenuation_1, Attenuation_7, Attenuation_8], properties {inputVolume [1.2,0.8,0.2,0.1]}}
 {id=Modulator_2, children [Mixer_0], properties {pitchMax=0.9, pitchMin=0.9}}
 {id=Modulator_3, children [WavePlayer_6], properties {pitchMax=0.9, pitchMin=0.9}}
 {id=WavePlayer_2, properties {bLooping=true, soundWaveAssetPtr=/Game/Athena/Sounds/Water/Ambient/Amb_Water_Ocean_Shore_Gentle_Laps_02}}
 {id=WavePlayer_6, properties {bLooping=true, soundWaveAssetPtr=/Game/Athena/Sounds/Water/Ambient/Amb_Water_Ocean_Waves_Med_Distant_Loop_01}}
]
properties {
 bOverrideConcurrency=true
 maxDistance=12000
 soundClassObject=&Fort_Ambient_Outdoors_Athena
 volumeMultiplier=0.501
 attenuationOverrides {
  bAttenuateWithLPF=true
  falloffDistance=6000
  lPFFrequencyAtMax=500
  nonSpatializedRadiusStart=700
  stereoSpread=1000
  attenuationShapeExtents {x=925}
 }
 concurrencyOverrides {maxCount=1, voiceStealReleaseTime=5}
}
